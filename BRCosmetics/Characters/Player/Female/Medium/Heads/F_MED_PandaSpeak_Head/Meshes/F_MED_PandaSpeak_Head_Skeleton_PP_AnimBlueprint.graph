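# Post-process AnimBlueprint for the PandaSpeak female head that moves the eyelashes with blinks.
# Each update, Remap Material Blinks takes the eye's dynamic material, cached on customization load or during update for front end and mannequins, and returns left and right blink values.
# These values set evaluation times in one sequence, layered onto each side's lash bones over the input pose.
# A local Blink Mapping function exists but is not called here.
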AnimBlueprint /BRCosmetics/Characters/Player/Female/Medium/Heads/F_MED_PandaSpeak_Head/Meshes/F_MED_PandaSpeak_Head_Skeleton_PP_AnimBlueprint
target_skeleton=&/F_MED_PandaSpeak_Head_Skeleton
graphs [
 AnimGraph {nodes [
   n1 = Layered blend per bone!(
    BasePose=Input Pose()
    BlendPoses_0="Sequence Evaluator (by time) 'F_MED_PandaSpeak_Head_Skeleton_Sequence'"!(ExplicitTime:float=Get L Blink Value() -> (L Blink Value:double)) -> (Pose:Pose Link Structure) {
     is=title AnimGraphNode_SequenceEvaluator
     sequence=&/F_MED_PandaSpeak_Head_Skeleton_Sequence
    }
    BlendPoses_1:Pose Link Structure="Sequence Evaluator (by time) 'F_MED_PandaSpeak_Head_Skeleton_Sequence'"(ExplicitTime=Get R Blink Value() -> (R Blink Value:double)) {sequence=&F_MED_PandaSpeak_Head_Skeleton_Sequence}
    BlendWeights_0
    BlendWeights_1:float=1
   ) -> (Pose) {
    blendPoses [{linkId=-1, sourceLinkId=-1}, {linkId=-1, sourceLinkId=-1}]
    blendWeights [1,1]
    layerSetup [
     {branchFilters [{boneName=deform_lash_01_l}, {boneName=deform_lash_02_l}, {boneName=deform_lash_03_l}]}
     {branchFilters [{boneName=deform_lash_01_r}, {boneName=deform_lash_02_r}, {boneName=deform_lash_03_r}]}
    ]
   }
   Output Pose(Result=@n1)
  ]}
 Blink Mapping {nodes [
   Blink Mapping!() -> (then, Material:Material Instance Dynamic Object Reference, Squint Influence:double=0.35, "Anim Length(Seconds)":double=3.3) {is=title K2Node_FunctionEntry}
   n1 = Get Material() -> (Material:Material Instance Dynamic Object Reference)
   n2 = GetCurveValue(CurveName=r_blink_pose)
   > SetScalarParameterValue(self=@n1, ParameterName=l_blink_pose, Value=@n2)
   > SetScalarParameterValue(self=@n1, ParameterName=r_blink_pose, Value=@n2)
   n3 = GetCurveValue(CurveName=R_upperLidDown_pose)
   > SetScalarParameterValue(self=@n1, ParameterName=L_upperLidDown_pose, Value=@n3)
   > SetScalarParameterValue(self=@n1, ParameterName=R_upperLidDown_pose, Value=@n3)
   n4 = GetCurveValue(CurveName=R_squint_inner_pose)
   > SetScalarParameterValue(self=@n1, ParameterName=L_squint_inner_pose, Value=@n4)
   > SetScalarParameterValue(self=@n1, ParameterName=R_squint_inner_pose, Value=@n4)
   n5 = float * float(A=@n4, B=Get Squint Influence() -> (Squint Influence:double)) {is=title K2Node_PromotableOperator}
   > Return Node!(
    execute
    Blink:double=MapRangeClamped(
     Value="Max (Float)"!(A=@n2, B=@n3, C:double=@n5) -> (ReturnValue)
     InRangeA=0
     InRangeB=1
     OutRangeB="Get Anim Length(Seconds)"() -> ("Anim Length(Seconds)":double)
    )
   )
  ]}
 EventGraph {nodes [
   Event BlueprintUpdateAnimation()
   > n1 = Sequence()
   > n2 = Branch(Condition=Get bIsInFrontEnd())
   n3 = Get Dynamic Material Instance(execute=[@n4.then, @n2.then], SkeletalMeshComponent=GetPartSkeletalMeshComponent())
   > Set Eye Mat(Eye Mat:Material Instance Dynamic Object Reference=@n3) -> (Output_Get:Material Instance Dynamic Object Reference)
   n4 = Cast To FortPlayerMannequin!(execute=@n2.else, Object=GetOwningActor()) -> (then, CastFailed, AsFort Player Mannequin:Fort Player Mannequin Object Reference) {is=title K2Node_DynamicCast}
   n5 = Remap Material Blinks!(
    execute=@n1.then_1
    Character Part Anim Instance:Anim Instance Object Reference=Self-Reference()
    Dynamic Material Instance:Material Instance Dynamic Object Reference=Get Eye Mat() -> (Eye Mat:Material Instance Dynamic Object Reference)
    Squint Influence:double=0.35
    "Anim Length(Seconds)":double=3.3
   ) -> (then, L Blink Value:double, R Blink Value:double, L Blink Lower Value:double, R Blink Lower Value:double)
   > Set L Blink Value(L Blink Value:double=@n5) -> (Output_Get:double)
   > Set R Blink Value(R Blink Value:double=@n5.R Blink Value) -> (Output_Get:double)
   Event OnCharacterAndPickaxeCustomizationLoaded!() -> (
    OutputDelegate
    then
    CustomBaseMesh:Skeletal Mesh Component Object Reference
    CustomParts:Array of Custom Character Part Object References
    CustomMeshParts:Array of Skeletal Mesh Component Budgeted Object References
   ) {is=title K2Node_Event}
   > n6 = Get Dynamic Material Instance(SkeletalMeshComponent=GetPartSkeletalMeshComponent())
   > Set Eye Mat(Eye Mat=@n6)
  ]}
]
variables [
 Eye Mat {type=MaterialInstanceDynamic}
 L Blink Value {type=number}
 R Blink Value {type=number}
]
